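# Editor utility widget that serves as a reusable primary-styled button for TechAudioTools editor interfaces.
# A single EditorUtilityButton uses rounded-box brushes in shades of blue for normal, hovered, pressed and disabled states, with a slight padding shift when pressed.
# A named slot called Content inside the button lets parent widgets supply the label or other content.
# The button's clicked, pressed and released events are re-broadcast through the widget's own matching event dispatchers.
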EditorUtilityWidgetBlueprint /TechAudioToolsContent/Widgets/Editor/Buttons/EUW_Button_Primary
graphs [
 EventGraph {nodes [
   "On Pressed (Button)"() {is=title K2Node_ComponentBoundEvent}
   > Call On Pressed!(execute:Exec, self:Tech Audio Tools Button Primary Object Reference) -> (then:Exec) {is=title K2Node_CallDelegate}
   "On Clicked (Button)"() {is=title K2Node_ComponentBoundEvent}
   > Call On Clicked!(execute:Exec, self:Tech Audio Tools Button Primary Object Reference) -> (then:Exec) {is=title K2Node_CallDelegate}
   "On Released (Button)"() {is=title K2Node_ComponentBoundEvent}
   > Call On Released!(execute:Exec, self:Tech Audio Tools Button Primary Object Reference) -> (then:Exec) {is=title K2Node_CallDelegate}
  ]}
 On Clicked {nodes [On Clicked() {is=title K2Node_FunctionEntry}]}
 On Pressed {nodes [On Pressed() {is=title K2Node_FunctionEntry}]}
 On Released {nodes [On Released() {is=title K2Node_FunctionEntry}]}
]
variables [
 On Clicked {default=()}
 On Pressed {default=()}
 On Released {default=()}
]
widgets {
 info {rootWidgetClass=&EditorUtilityButton, widgetCount=2}
 widgets [
  Button {
   widgetClassPath=&EditorUtilityButton
   widget=Button
   properties {widgetStyle {
     clickedSlateSound {resourceObject=null}
     disabled {
      bIsDynamicallyLoaded=false
      drawAs=RoundedBox
      imageType=NoImage
      mirroring=NoMirror
      resourceName=None
      resourceObject=null
      tiling=NoTile
      imageSize=<32 32>
      margin=<ltrb 0 0 0 0>
      outlineSettings {
       bUseBrushTransparency=false
       roundingType=FixedRadius
       width=1
       color {colorUseRule=UseColor_Specified, specifiedColor=<rgba 0.006 0.006 0.006 1>}
       cornerRadii=<4 4 4 4>
      }
      tintColor {colorUseRule=UseColor_Specified, specifiedColor=<rgba 0 0.0849 0.3906 1>}
      uVRegion {bIsValid=false, max=<0 0>, min=<0 0>}
     }
     disabledForeground {colorUseRule=UseColor_ColorTable, specifiedColor=<rgba 0 0 0 0>}
     hovered {
      bIsDynamicallyLoaded=false
      drawAs=RoundedBox
      imageType=NoImage
      mirroring=NoMirror
      resourceName=None
      resourceObject=null
      tiling=NoTile
      imageSize=<32 32>
      margin=<ltrb 0 0 0 0>
      outlineSettings {
       bUseBrushTransparency=false
       roundingType=FixedRadius
       width=1
       color {colorUseRule=UseColor_Specified, specifiedColor=<rgba 0.006 0.006 0.006 1>}
       cornerRadii=<4 4 4 4>
      }
      tintColor {colorUseRule=UseColor_Specified, specifiedColor=<rgba 0.0044 0.2384 1 1>}
      uVRegion {bIsValid=false, max=<0 0>, min=<0 0>}
     }
     hoveredForeground {colorUseRule=UseColor_ColorTable, specifiedColor=<rgba 0 0 0 0>}
     hoveredSlateSound {resourceObject=null}
     normal {
      bIsDynamicallyLoaded=false
      drawAs=RoundedBox
      imageType=NoImage
      mirroring=NoMirror
      resourceName=None
      resourceObject=null
      tiling=NoTile
      imageSize=<32 32>
      margin=<ltrb 0 0 0 0>
      outlineSettings {
       bUseBrushTransparency=false
       roundingType=FixedRadius
       width=1
       color {colorUseRule=UseColor_Specified, specifiedColor=<rgba 0.006 0.006 0.006 1>}
       cornerRadii=<4 4 4 4>
      }
      tintColor {colorUseRule=UseColor_Specified, specifiedColor=<rgba 0 0.162 0.7454 1>}
      uVRegion {bIsValid=false, max=<0 0>, min=<0 0>}
     }
     normalForeground {colorUseRule=UseColor_ColorTable, specifiedColor=<rgba 0 0 0 0>}
     normalPadding=<ltrb 2 0 2 0>
     pressed {
      bIsDynamicallyLoaded=false
      drawAs=RoundedBox
      imageType=NoImage
      mirroring=NoMirror
      resourceName=None
      resourceObject=null
      tiling=NoTile
      imageSize=<32 32>
      margin=<ltrb 0 0 0 0>
      outlineSettings {
       bUseBrushTransparency=false
       roundingType=FixedRadius
       width=1
       color {colorUseRule=UseColor_Specified, specifiedColor=<rgba 0.006 0.006 0.006 1>}
       cornerRadii=<4 4 4 4>
      }
      tintColor {colorUseRule=UseColor_Specified, specifiedColor=<rgba 0 0.0802 0.3515 1>}
      uVRegion {bIsValid=false, max=<0 0>, min=<0 0>}
     }
     pressedForeground {colorUseRule=UseColor_ColorTable, specifiedColor=<rgba 0 0 0 0>}
     pressedPadding=<ltrb 2 0.5 2 -0.5>
     pressedSlateSound {resourceObject=null}
    }}
   slot=~
  }
  Content {widgetClassPath=&NamedSlot, parent=Button, widget=Content, slot {class=ButtonSlot}}
 ]
}
