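# Editor Utility Widget for a reusable primary button with an icon, used in TechAudioTools content editor UI.
# It nests an editor Slate icon inside the named slot of an EUW_Button_Primary via a vertical box.
# PreConstruct applies the exposed brush path, size and color variables to the icon.
# The inner button's clicked, pressed and released events are re-broadcast through the widget's own dispatchers.
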EditorUtilityWidgetBlueprint /TechAudioToolsContent/Widgets/Editor/Buttons/EUW_Button_Primary_Icon
graphs [
 EventGraph {nodes [
   "On Clicked (Button_Primary)"() {is=title K2Node_ComponentBoundEvent}
   > Call On Clicked!(execute:Exec, self:Tech Audio Tools Button Primary Icon Object Reference) -> (then:Exec) {is=title K2Node_CallDelegate}
   "On Pressed (Button_Primary)"() {is=title K2Node_ComponentBoundEvent}
   > Call On Pressed!(execute:Exec, self:Tech Audio Tools Button Primary Icon Object Reference) -> (then:Exec) {is=title K2Node_CallDelegate}
   "On Released (Button_Primary)"() {is=title K2Node_ComponentBoundEvent}
   > Call On Released!(execute:Exec, self:Tech Audio Tools Button Primary Icon Object Reference) -> (then:Exec) {is=title K2Node_CallDelegate}
   Event PreConstruct()
   > SetIconBrushByPath!(
    execute
    self:Tech Audio Tools Editor Slate Icon Object Reference=Get Icon() -> (Icon:Tech Audio Tools Editor Slate Icon Object Reference)
    InFullIconPath:Name=Get Icon Brush Path() -> (Icon Brush Path:Name=None)
   ) -> (then)
   > SetIconSize!(
    execute
    self:Tech Audio Tools Editor Slate Icon Object Reference=Get Icon()
    InSize:Vector 2D Structure=Get Icon Size() -> (Icon Size:Vector 2D Structure)
   ) -> (then)
   > SetIconColor!(
    execute
    self:Tech Audio Tools Editor Slate Icon Object Reference=Get Icon()
    InColor:Linear Color Structure=Get Icon Color() -> (Icon Color:Linear Color Structure)
   ) -> (then)
  ]}
 On Clicked {nodes [On Clicked() {is=title K2Node_FunctionEntry}]}
 On Pressed {nodes [On Pressed() {is=title K2Node_FunctionEntry}]}
 On Released {nodes [On Released() {is=title K2Node_FunctionEntry}]}
]
variables [
 Icon Brush Path {type=string, default=EditorStyle.Icons.Save}
 Icon Color {type=LinearColor, default=<rgba 1 1 1 1>}
 Icon Size {type=Vector2D, default=<20 20>}
 On Clicked {default=()}
 On Pressed {default=()}
 On Released {default=()}
]
widgets {
 info {namedSlotCount=1, rootWidgetClass=&/EUW_Button_Primary.EUW_Button_Primary_C, widgetCount=3}
 widgets [
  Button_Primary {
   widgetClassPath=&/EUW_Button_Primary
   bIsVariable=true
   widget=Button_Primary
   properties {bOverride_Cursor=true}
  }
  VerticalBox_0 {widgetClassPath=&VerticalBox, namedSlotHost=Button_Primary, widget=VerticalBox_0}
  Icon {
   widgetClassPath=/Script/TechAudioToolsContentEditor.TechAudioToolsEditorSlateIcon
   bIsVariable=true
   parent=VerticalBox_0
   widget=Icon
   slot {class=VerticalBoxSlot, properties {padding=<ltrb 3 4 3 4>}}
  }
 ]
}
