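# Editor utility widget that provides a secondary-style button with a text label for Tech Audio Tools editor UI.
# It places a padded TextBlock in the named slot of a nested EUW_Button_Secondary.
# The inner button's click, press and release events are re-broadcast through the widget's own dispatchers.
# On release, it first sets the label's color to the Text Color variable, which defaults to white.
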EditorUtilityWidgetBlueprint /TechAudioToolsContent/Widgets/Editor/Buttons/EUW_Button_Secondary_Text
graphs [
 EventGraph {nodes [
   "On Clicked (Button_Secondary)"() {is=title K2Node_ComponentBoundEvent}
   > Call On Clicked!(execute:Exec, self:Tech Audio Tools Button Secondary Text Object Reference) -> (then:Exec) {is=title K2Node_CallDelegate}
   "On Pressed (Button_Secondary)"() {is=title K2Node_ComponentBoundEvent}
   > Call On Pressed!(execute:Exec, self:Tech Audio Tools Button Secondary Text Object Reference) -> (then:Exec) {is=title K2Node_CallDelegate}
   "On Released (Button_Secondary)"() {is=title K2Node_ComponentBoundEvent}
   > Set ColorAndOpacity!(
    execute
    ColorAndOpacity:Slate Color Structure=Make Slate Color(SpecifiedColor=Get Text Color() -> (Text Color:Linear Color Structure))
    self:Text Object Reference=Get Text!() -> (Text:Text Object Reference)
   ) -> (then, Output_Get:Slate Color Structure)
   > Call On Released!(execute:Exec, self:Tech Audio Tools Button Secondary Text Object Reference) -> (then:Exec) {is=title K2Node_CallDelegate}
  ]}
 On Clicked {nodes [On Clicked() {is=title K2Node_FunctionEntry}]}
 On Pressed {nodes [On Pressed() {is=title K2Node_FunctionEntry}]}
 On Released {nodes [On Released() {is=title K2Node_FunctionEntry}]}
]
variables [
 On Clicked {default=()}
 On Pressed {default=()}
 On Released {default=()}
 Text Color {type=LinearColor, default=<rgba 1 1 1 1>}
]
widgets {
 info {namedSlotCount=1, rootWidgetClass=&/EUW_Button_Secondary.EUW_Button_Secondary_C, widgetCount=3}
 widgets [
  Button_Secondary {
   widgetClassPath=&/EUW_Button_Secondary
   bIsVariable=true
   widget=Button_Secondary
   properties {bOverride_Cursor=true}
  }
  VerticalBox_0 {widgetClassPath=&VerticalBox, namedSlotHost=Button_Secondary, widget=VerticalBox_0}
  Text {
   widgetClassPath=&TextBlock
   bIsVariable=true
   parent=VerticalBox_0
   widget=Text
   properties {
    text=Button
    font {
     bForceMonospaced=false
     bMaterialIsStencil=false
     fontMaterial=null
     fontObject=&Roboto
     letterSpacing=0
     monospacedWidth=1
     size=9
     skewAmount=0
     typefaceFontName=Regular
     outlineSettings {
      bApplyOutlineToDropShadows=false
      bMiteredCorners=false
      bSeparateFillAlpha=false
      outlineMaterial=null
      outlineSize=0
      outlineColor=<rgba 0 0 0 1>
     }
    }
   }
   slot {class=VerticalBoxSlot, properties {padding=<ltrb 10 2 10 2>}}
  }
 ]
}
